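# Physics asset for the Supersonic Pink two-handed pickaxe mesh, setting up a capsule handle body with sphere bodies on its dynamic bones.
# A roll-then-tilt chain off the handle splits single-axis swing between two joints and carries four wheel bodies on tightly limited constraints.
# A second chain runs from a twist-only end joint on the handle through end roll and end tilt, with much wider swing limits.
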PhysicsAsset /BRCosmetics/Weapons/FORT_Melee/Pickaxe_Supersonic_Pink/Meshes/Supersonic_Pink_Axe_Physics
bodies {
 Melee_TwoHanded_Handle [{shapeType=Capsule, length=58.9297, radius=7.0347, center=<0.5669 -0.3386 14.1406>, rotation=<rot -0.0223 -90 -0.0078>}]
 dyn_end_jnt [{shapeType=Sphere, radius=8, shapeName=dyn_end_jnt_sphere, center=<-0.391 -4.7309 -0.0011>}]
 dyn_end_roll [{shapeType=Sphere, radius=8, shapeName=dyn_end_roll_sphere, center=<5 0 0>}]
 dyn_end_tilt [{shapeType=Sphere, radius=8, shapeName=dyn_end_tilt_sphere, center=<5 0 0>}]
 dyn_roll [{shapeType=Sphere, radius=5, shapeName=dyn_roll_sphere, center=<1 0 0>}]
 dyn_tilt [{shapeType=Sphere, radius=5, shapeName=dyn_tilt_sphere, center=<1 0 0>}]
 dyn_wheel_lf [{shapeType=Sphere, radius=8, shapeName=dyn_wheel_lf_sphere, center=<7 0 0>}]
 dyn_wheel_lr [{shapeType=Sphere, radius=8, shapeName=dyn_wheel_lr_sphere, center=<7 0 0>}]
 dyn_wheel_rf [{shapeType=Sphere, radius=8, shapeName=dyn_wheel_rf_sphere, center=<7 0 0>}]
 dyn_wheel_rr [{shapeType=Sphere, radius=8, shapeName=dyn_wheel_rr_sphere, center=<7 0 0>}]
}
constraints [
 {bone1Name=dyn_wheel_lf, bone2Name=dyn_tilt, swing1LimitDegrees=1, swing2LimitDegrees=10, twistLimitDegrees=1}
 {bone1Name=dyn_wheel_rf, bone2Name=dyn_tilt, swing1LimitDegrees=1, swing2LimitDegrees=10, twistLimitDegrees=1}
 {bone1Name=dyn_wheel_lr, bone2Name=dyn_tilt, swing1LimitDegrees=1, swing2LimitDegrees=10, twistLimitDegrees=1}
 {bone1Name=dyn_wheel_rr, bone2Name=dyn_tilt, swing1LimitDegrees=1, swing2LimitDegrees=10, twistLimitDegrees=1}
 {
  bone1Name=dyn_roll
  bone2Name=Melee_TwoHanded_Handle
  swing1LimitDegrees=1
  swing1Motion=Locked
  swing2LimitDegrees=4.2
  twistLimitDegrees=1
  twistMotion=Locked
 }
 {
  bone1Name=dyn_tilt
  bone2Name=dyn_roll
  swing1LimitDegrees=4.2
  swing2LimitDegrees=0
  swing2Motion=Locked
  twistLimitDegrees=1
  twistMotion=Locked
 }
 {
  bone1Name=dyn_end_roll
  bone2Name=dyn_end_jnt
  swing1LimitDegrees=1
  swing1Motion=Locked
  swing2LimitDegrees=30
  twistLimitDegrees=1
  twistMotion=Locked
 }
 {bone1Name=dyn_end_tilt, bone2Name=dyn_end_roll, swing1LimitDegrees=90, swing2LimitDegrees=1, twistLimitDegrees=1}
 {
  bone1Name=dyn_end_jnt
  bone2Name=Melee_TwoHanded_Handle
  swing1LimitDegrees=0
  swing1Motion=Locked
  swing2LimitDegrees=0
  swing2Motion=Locked
  twistLimitDegrees=120
 }
]
